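# Post-process style AnimBP for the F_MED_StarStray player body, adding secondary motion and corrective posing to pants, belt, backpack and cape attach bones.
# AnimGraph copies the parent mesh pose through a deform Control Rig, then chains AnimDynamics pants, belt constraints, look-ats and additive bone offsets.
# EventGraph drives those alphas from thigh-to-pelvis deltas, movement states and helmet detection.
# It also feeds cloth WPO and skydive wind parameters to dynamic materials.
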AnimBlueprint /Game/Characters/Player/Female/Medium/Bodies/F_MED_StarStray/Meshes/F_MED_StarStray_AnimBP
target_skeleton=&/F_MED_StarStray_Skeleton
graphs [
 AnimGraph {nodes [
   "Constraint - corrective_belt_f_l (T)"(Alpha=0.5, ConstraintWeights_0=1) {
    alphaBoolBlend {blendInTime=0.2, blendOption=Linear, blendOutTime=0.2}
    boneToModify {boneName=corrective_belt_f_l}
    constraintSetup [{transformType=Translation, targetBone {boneName=thigh_l}}]
    constraintWeights [1]
   }
   n1 = Control Rig(Source=Copy Pose From Mesh() {bUseAttachedParent=true}) {
    alphaBoolBlend {blendOption=Linear}
    controlRigAssetReference {blueprintRigClass=&/F_MED_StarStray_DeformRig.F_MED_StarStray_DeformRig_C}
   }
   n2 = Get pants alpha dynamics() -> (pants alpha dynamics:double)
   n3 = Anim Dynamics - Bone: dyn_l_pants(ComponentPose=Local To Component(LocalPose=@n1), Alpha=@n2, ExternalForce="0,0,0") {
    angularBiasOverride=0.8
    angularDampingOverride=0.9
    angularSpringConstant=250
    bLinearSpring=true
    bOverrideLinearDamping=true
    bUseGravityOverride=true
    linearDampingOverride=1
    linearSpringConstant=80
    alphaBoolBlend {blendOption=Linear}
    boundBone {boneName=dyn_l_pants}
    componentAppliedLinearAccClamp=<3000 3000 3000>
    componentLinearAccScale=<0.1 0.1 0.1>
    componentLinearVelScale=<0.9 0.9 0.9>
    physicsBodyDefinitions [{
      boundBone {boneName=dyn_l_pants}
      boxExtents=<20 20 20>
      constraintSetup {angularLimitsMax=<0 0 0>, angularLimitsMin=<0 0 0>, angularTarget=<0 0 0>, linearAxesMax=<30 4 5>, linearAxesMin=<-30 -4 -5>}
      localJointOffset=<0 0 0>
     }]
   }
   n4 = Anim Dynamics - Bone: dyn_r_pants(ComponentPose=@n3, Alpha=@n2, ExternalForce="0,0,0") {
    angularBiasOverride=0.8
    angularDampingOverride=0.9
    angularSpringConstant=250
    bLinearSpring=true
    bOverrideLinearDamping=true
    bUseGravityOverride=true
    linearDampingOverride=1
    linearSpringConstant=80
    alphaBoolBlend {blendOption=Linear}
    boundBone {boneName=dyn_r_pants}
    componentAppliedLinearAccClamp=<3000 3000 3000>
    componentLinearAccScale=<0.1 0.1 0.1>
    componentLinearVelScale=<0.9 0.9 0.9>
    physicsBodyDefinitions [{
      boundBone {boneName=dyn_r_pants}
      boxExtents=<20 20 20>
      constraintSetup {angularLimitsMax=<0 0 0>, angularLimitsMin=<0 0 0>, angularTarget=<0 0 0>, linearAxesMax=<30 4 5>, linearAxesMin=<-30 -4 -5>}
      localJointOffset=<0 0 0>
     }]
   }
   n5 = "Constraint - corrective_belt_r (PP)"!(ComponentPose=@n4, Alpha=1, ConstraintWeights_0=0.2, ConstraintWeights_1:float=0.8) -> (Pose) {
    alphaBoolBlend {blendOption=Linear}
    boneToModify {boneName=corrective_belt_r}
    constraintSetup [{targetBone {boneName=pelvis}}, {targetBone {boneName=thigh_twist_01_r}}]
    constraintWeights [1,1]
   }
   n6 = "Constraint - corrective_belt_l (PP)"!(ComponentPose=@n5, Alpha=Get left part belt() -> (left part belt:float), ConstraintWeights_0=0.3, ConstraintWeights_1:float=0.7) -> (Pose) {
    alphaBoolBlend {blendOption=Linear}
    alphaScaleBiasClamp {bInterpResult=true}
    boneToModify {boneName=corrective_belt_l}
    constraintSetup [{targetBone {boneName=pelvis}}, {targetBone {boneName=thigh_twist_01_l}}]
    constraintWeights [1,1]
   }
   n7 = "Constraint - corrective_belt_b_r (PP)"!(ComponentPose=@n6, Alpha=1, ConstraintWeights_0=0.82, ConstraintWeights_1:float=0.18) -> (Pose) {
    alphaBoolBlend {blendOption=Linear}
    boneToModify {boneName=corrective_belt_b_r}
    constraintSetup [{targetBone {boneName=pelvis}}, {targetBone {boneName=thigh_twist_01_r}}]
    constraintWeights [1,1]
   }
   n8 = "Constraint - corrective_belt_r_l (PP)"!(ComponentPose=@n7, Alpha=1, ConstraintWeights_0=0.85, ConstraintWeights_1:float=0.15) -> (Pose) {
    alphaBoolBlend {blendOption=Linear}
    boneToModify {boneName=corrective_belt_r_l}
    constraintSetup [{targetBone {boneName=pelvis}}, {targetBone {boneName=thigh_twist_01_l}}]
    constraintWeights [1,1]
   }
   n9 = "Transform (Modify) Bone - Bone: corrective_belt_r_l"!(ComponentPose=@n8, bAlphaBoolEnabled:bool=True, Translation="0,-1.5,0", Rotation="0,0,0", Scale="1,1,1") -> (Pose) {
    alphaInputType=Bool
    rotationSpace=BCS_BoneSpace
    translationMode=BMM_Additive
    alphaBoolBlend {blendInTime=0.2, blendOption=Linear, blendOutTime=0.2}
    boneToModify {boneName=corrective_belt_r_l}
    translation {y=-1.5}
   }
   n10 = "Transform (Modify) Bone - Bone: corrective_belt_b"!(ComponentPose=@n9, bAlphaBoolEnabled:bool=True, Translation="0,1,0", Rotation="0,0,0", Scale="1,1,1") -> (Pose) {
    alphaInputType=Bool
    rotationSpace=BCS_BoneSpace
    translationMode=BMM_Additive
    alphaBoolBlend {blendInTime=0.2, blendOption=Linear, blendOutTime=0.2}
    boneToModify {boneName=corrective_belt_b}
    translation {y=1}
   }
   n11 = "Constraint - corrective_belt_b (PPP)"!(ComponentPose=@n10, Alpha=1, ConstraintWeights_0=0.8, ConstraintWeights_1:float=0.1, ConstraintWeights_2:float=0.1) -> (Pose) {
    alphaBoolBlend {blendOption=Linear}
    boneToModify {boneName=corrective_belt_b}
    constraintSetup [{targetBone {boneName=pelvis}}, {targetBone {boneName=thigh_twist_01_l}}, {targetBone {boneName=thigh_twist_01_r}}]
    constraintWeights [1,1,1]
   }
   n12 = "Transform (Modify) Bone - Bone: corrective_belt_l"!(ComponentPose=@n11, bAlphaBoolEnabled:bool=True, Translation="0,-4,0", Rotation="2,0,0", Scale="1,1,1") -> (Pose) {
    alphaInputType=Bool
    rotationMode=BMM_Additive
    rotationSpace=BCS_BoneSpace
    translationMode=BMM_Additive
    translationSpace=BCS_BoneSpace
    alphaBoolBlend {blendInTime=0.2, blendOption=Linear, blendOutTime=0.2}
    boneToModify {boneName=corrective_belt_l}
    rotation {pitch=2}
    translation {y=-4}
   }
   n13 = "Transform (Modify) Bone - Bone: corrective_belt_b"(ComponentPose=@n12, bAlphaBoolEnabled=True, Translation="0.0,-2,0.0", Rotation="0,0,0", Scale="1,1,1") {
    alphaInputType=Bool
    rotationMode=BMM_Additive
    rotationSpace=BCS_BoneSpace
    translationMode=BMM_Additive
    translationSpace=BCS_BoneSpace
    alphaBoolBlend {blendInTime=0.2, blendOption=Linear, blendOutTime=0.2}
    boneToModify {boneName=corrective_belt_b}
    translation {y=-2}
   }
   n14 = Look At - Bone: corrective_belt_f_r_2(ComponentPose=@n13, Alpha=1) {
    alphaBoolBlend {blendOption=Linear}
    boneToModify {boneName=corrective_belt_f_r_2}
    lookAtLocation {y=-20}
    lookAtTarget {boneReference {boneName=thigh_r}}
    lookAt_Axis {axis=<-1 0>}
   }
   n15 = Look At - Bone: corrective_belt_f_l_2(ComponentPose=@n14, Alpha=1) {
    alphaBoolBlend {blendOption=Linear}
    boneToModify {boneName=corrective_belt_f_l_2}
    lookAtLocation=<-100 20>
    lookAtTarget {boneReference {boneName=thigh_l}}
    lookAt_Axis {axis=<-1 0>}
    lookUp_Axis {axis {x=-1}}
   }
   n16 = Get right leg alpha_0() -> (right leg alpha_0:float)
   n17 = "Transform (Modify) Bone - Bone: corrective_belt_f_r"(ComponentPose=@n15, Alpha=@n16, Translation="11,1.5,0", Rotation="0,0,0", Scale="1,1,1") {
    rotationMode=BMM_Additive
    rotationSpace=BCS_BoneSpace
    translationMode=BMM_Additive
    translationSpace=BCS_BoneSpace
    alphaBoolBlend {blendOption=Linear}
    boneToModify {boneName=corrective_belt_f_r}
    translation=<11 1.5>
   }
   n18 = "Transform (Modify) Bone - Bone: corrective_belt_f_r_2"(ComponentPose=@n17, Alpha=@n16, Translation="5,0,0", Rotation="0,-20,0", Scale="1,1,1") {
    rotationMode=BMM_Additive
    rotationSpace=BCS_BoneSpace
    translationMode=BMM_Additive
    translationSpace=BCS_BoneSpace
    alphaBoolBlend {blendInTime=0.2, blendOption=Linear, blendOutTime=0.2}
    boneToModify {boneName=corrective_belt_f_r_2}
    rotation {yaw=-20}
    translation {x=5}
   }
   n19 = "Transform (Modify) Bone - Bone: corrective_belt_f_r"(
    ComponentPose=@n18
    Alpha=Get right leg alpha combp crouching sprinting() -> (right leg alpha combp crouching sprinting:double)
    Translation="3,0,0"
    Rotation="10,0,0"
    Scale="1,1,1"
   ) {
    rotationMode=BMM_Additive
    rotationSpace=BCS_BoneSpace
    translationMode=BMM_Additive
    translationSpace=BCS_BoneSpace
    alphaBoolBlend {blendInTime=0.2, blendOption=Linear, blendOutTime=0.2}
    boneToModify {boneName=corrective_belt_f_r}
    rotation {pitch=10}
    translation {x=3}
   }
   n20 = "Transform (Modify) Bone - Bone: corrective_belt_f_l"(ComponentPose=@n19, Alpha=Get left leg alpha_0() -> (left leg alpha_0:float), Translation="12.5,1,0", Rotation="0,0,0", Scale="1,1,1") {
    rotationMode=BMM_Additive
    rotationSpace=BCS_BoneSpace
    translationMode=BMM_Additive
    translationSpace=BCS_BoneSpace
    alphaBoolBlend {blendOption=Linear}
    boneToModify {boneName=corrective_belt_f_l}
    translation=<12.5 1>
   }
   n21 = Get left leg alpha combo crouching sitting() -> (left leg alpha combo crouching sitting:double)
   n22 = "Transform (Modify) Bone - Bone: corrective_belt_f_l"(ComponentPose=@n20, Alpha=@n21, Translation="0,0,0", Rotation="0,40,0", Scale="1,1,1") {
    rotationMode=BMM_Additive
    rotationSpace=BCS_BoneSpace
    translationMode=BMM_Additive
    translationSpace=BCS_BoneSpace
    alphaBoolBlend {blendInTime=0.2, blendOption=Linear, blendOutTime=0.2}
    alphaScaleBiasClamp {interpSpeedDecreasing=0.5, interpSpeedIncreasing=0.5}
    boneToModify {boneName=corrective_belt_f_l}
    rotation {yaw=40}
   }
   n23 = "Transform (Modify) Bone - Bone: corrective_belt_f_l_2"(ComponentPose=@n22, Alpha=@n21, Translation="0,2,0", Rotation="0,30,0", Scale="1,1,1") {
    rotationMode=BMM_Additive
    rotationSpace=BCS_BoneSpace
    translationMode=BMM_Additive
    translationSpace=BCS_BoneSpace
    alphaBoolBlend {blendInTime=0.2, blendOption=Linear, blendOutTime=0.2}
    boneToModify {boneName=corrective_belt_f_l_2}
    rotation {yaw=30}
    translation {y=2}
   }
   n24 = "Transform (Modify) Bone - Bone: corrective_belt_f_l"(
    ComponentPose=@n23
    Alpha=Get left leg alpha combo crouching sprinting() -> (left leg alpha combo crouching sprinting:double)
    Translation="3.5,0,0"
    Rotation="0,0,0"
    Scale="1,1,1"
   ) {
    rotationMode=BMM_Additive
    rotationSpace=BCS_BoneSpace
    translationMode=BMM_Additive
    translationSpace=BCS_BoneSpace
    alphaBoolBlend {blendInTime=0.2, blendOption=Linear, blendOutTime=0.2}
    boneToModify {boneName=corrective_belt_f_l}
    translation {x=3.5}
   }
   n25 = "Constraint - corrective_belt_f_l (T)"(ComponentPose=@n24, Alpha=Get crouching left float 0-5() -> (crouching left float 0-5:double), ConstraintWeights_0=1) {
    alphaBoolBlend {blendInTime=0.2, blendOption=Linear, blendOutTime=0.2}
    alphaScaleBiasClamp {bInterpResult=true}
    boneToModify {boneName=corrective_belt_f_l}
    constraintSetup [{transformType=Translation, targetBone {boneName=thigh_l}}]
    constraintWeights [1]
   }
   n26 = Get crouching left() -> (crouching left:bool)
   n27 = "Transform (Modify) Bone - Bone: corrective_belt_f_l"!(ComponentPose=@n25, bAlphaBoolEnabled:bool=@n26, Translation="1.3,1,0", Rotation="0,0,0", Scale="1,1,1") -> (Pose) {
    alphaInputType=Bool
    rotationMode=BMM_Additive
    rotationSpace=BCS_BoneSpace
    scaleMode=BMM_Additive
    scaleSpace=BCS_BoneSpace
    translationMode=BMM_Additive
    translationSpace=BCS_BoneSpace
    alphaBoolBlend {blendInTime=0.2, blendOption=Linear, blendOutTime=0.2}
    boneToModify {boneName=corrective_belt_f_l}
    translation=<1.3 1>
   }
   n28 = "Transform (Modify) Bone - Bone: corrective_belt_f_l_2"!(ComponentPose=@n27, bAlphaBoolEnabled:bool=@n26, Translation="0,0,0", Rotation="0,-25,0", Scale="1,1,1") -> (Pose) {
    alphaInputType=Bool
    rotationMode=BMM_Additive
    rotationSpace=BCS_BoneSpace
    translationMode=BMM_Additive
    translationSpace=BCS_BoneSpace
    alphaBoolBlend {blendInTime=0.2, blendOption=Linear, blendOutTime=0.2}
    boneToModify {boneName=corrective_belt_f_l_2}
    rotation {yaw=-25}
   }
   n29 = Get bIsSkyDiving()
   n30 = "Transform (Modify) Bone - Bone: attach_backpack"!(ComponentPose=@n28, bAlphaBoolEnabled:bool=@n29, Translation="0,0,0", Rotation="0,0,12", Scale="1,1,1") -> (Pose) {
    alphaInputType=Bool
    rotationMode=BMM_Additive
    rotationSpace=BCS_BoneSpace
    translationMode=BMM_Additive
    translationSpace=BCS_BoneSpace
    alphaBoolBlend {blendInTime=0.2, blendOption=Sinusoidal, blendOutTime=0.2}
    boneToModify {boneName=attach_backpack}
    rotation {roll=12}
   }
   n31 = "Transform (Modify) Bone - Bone: corrective_belt_b"(ComponentPose=@n30, bAlphaBoolEnabled=@n29, Translation="0,1,0", Rotation="0,0,0", Scale="1,1,1") {
    alphaInputType=Bool
    rotationMode=BMM_Additive
    rotationSpace=BCS_BoneSpace
    translationMode=BMM_Additive
    translationSpace=BCS_BoneSpace
    alphaBoolBlend {blendInTime=0.2, blendOption=Sinusoidal, blendOutTime=0.2}
    boneToModify {boneName=corrective_belt_b}
    translation {y=1}
   }
   n32 = "Transform (Modify) Bone - Bone: corrective_belt_b_r"!(ComponentPose=@n31, bAlphaBoolEnabled:bool=@n29, Translation="0,1,0", Rotation="0,0,0", Scale="1,1,1") -> (Pose) {
    alphaInputType=Bool
    rotationMode=BMM_Additive
    rotationSpace=BCS_BoneSpace
    translationMode=BMM_Additive
    translationSpace=BCS_BoneSpace
    alphaBoolBlend {blendInTime=0.2, blendOption=Sinusoidal, blendOutTime=0.2}
    boneToModify {boneName=corrective_belt_b_r}
    translation {y=1}
   }
   n33 = "Transform (Modify) Bone - Bone: attach_backpack"(
    ComponentPose=@n32
    bAlphaBoolEnabled=Get backpack move after jumping() -> (backpack move after jumping:bool)
    Translation="0,-1,0"
    Rotation="0,0,9"
    Scale="1,1,1"
   ) {
    alphaInputType=Bool
    rotationMode=BMM_Additive
    rotationSpace=BCS_BoneSpace
    translationMode=BMM_Additive
    translationSpace=BCS_BoneSpace
    alphaBoolBlend {blendInTime=0.2, blendOption=Linear, blendOutTime=0.2}
    boneToModify {boneName=attach_backpack}
    rotation {roll=9}
    translation {y=-1}
   }
   n34 = "Transform (Modify) Bone - Bone: attach_backpack"(ComponentPose=@n33, bAlphaBoolEnabled=Get is helmet on() -> (is helmet on:bool), Translation="0,3,0", Rotation="0,0,0", Scale="1,1,1") {
    alphaInputType=Bool
    translationMode=BMM_Additive
    translationSpace=BCS_BoneSpace
    alphaBoolBlend {blendOption=Linear}
    boneToModify {boneName=attach_backpack}
    translation {y=3}
   }
   n35 = "Transform (Modify) Bone - Bone: attach_cape"!(ComponentPose=@n34, bAlphaBoolEnabled:bool=Get is helmet on(), Translation="0,4,1", Rotation="0,0,0", Scale="1,1,1") -> (Pose) {
    alphaInputType=Bool
    translationMode=BMM_Additive
    translationSpace=BCS_BoneSpace
    alphaBoolBlend {blendOption=Linear}
    boneToModify {boneName=attach_cape}
    translation {y=4, z=1}
   }
   Output Pose(Result=Component To Local(ComponentPose=@n35)) {layerGroup=None}
  ]}
 EventGraph {nodes [
   n1 = Event BlueprintUpdateAnimation()
   > Set delta time x(delta time x:float=@n1.DeltaTimeX) -> (Output_Get:float)
   > n2 = Sequence!(execute) -> (then_0, then_1, then_2:Exec, then_3:Exec, then_4:Exec, then_5:Exec)
   n3 = OR Boolean!(A=Get bIsJumping(), B=Get bIsFalling(), C:bool=Get bIsCrouching()) -> (ReturnValue)
   > Set alpha pants(alpha pants:double=FInterpTo(
     Current=Get alpha pants() -> (alpha pants:double)
     Target=SelectFloat(A=1, B=0.8, bPickA=@n3)
     DeltaTime=Get delta time x() -> (delta time x:float)
     InterpSpeed=10
    )) -> (Output_Get:double)
   n4 = Get bIsCrouching()
   > Set left part belt(left part belt:float="To Float (Boolean)"(InBool=NOT Boolean(A=@n4))) -> (Output_Get:float)
   > Set jumping or running(jumping or running:bool=OR Boolean(A=Get bIsJumping(), B=Get bIsFalling(), C=Get bIsMoving2D())) -> (Output_Get:bool)
   > Set right leg alpha_0(right leg alpha_0:float=GetAlphaFromReferenceTransformDelta(
     SkeletalMeshComponent=Get skeletal mesh() -> (skeletal mesh:Skeletal Mesh Component Object Reference)
     BoneName=pelvis
     BaseName=thigh_r
     InputRange=(X=0,Y=97)
     TwistAxis="1,0,0"
    )) -> (Output_Get:float)
   > Set left leg alpha_0(left leg alpha_0:float=GetAlphaFromReferenceTransformDelta(SkeletalMeshComponent=Get skeletal mesh(), BoneName=pelvis, BaseName=thigh_l, InputRange=(X=0,Y=97), TwistAxis="1,0,0")) -> (Output_Get:float)
   > Set crouching or moving 2d(crouching or moving 2d:bool=OR Boolean!(A=Get bIsMoving2D(), B=Get bIsCrouching()) -> (ReturnValue)) -> (Output_Get:bool)
   > n5 = Branch(Condition=OR Boolean!(A=Get bIsInFrontEnd(), B=Get bIsMoving2D(), C:bool=Get bIsPlayingEmote()) -> (ReturnValue))
   > Set pants alpha dynamics(pants alpha dynamics:double=0.3) -> (Output_Get:double)
   Set pants alpha dynamics(execute=@n5.else, pants alpha dynamics=1)
   Sequence!(execute=@n2.then_3) -> (then_0, then_1)
   > Is Valid(InputObject=Get PlayerPawn() -> (PlayerPawn:Fort Player Pawn Athena Object Reference))
   > Set WPO Globals Function(
    Material=Get Mat Instance Dynamic()
    Overlay Material=Get Overlay Mat Instance Dynamic()
    Skeletal Mesh=Get skeletal mesh()
    "Local Noise Transform  Location"=Get Local Noise Transform Location() -> (Local Noise Transform Location:Vector)
    Local Noise Offset=Get Local Noise Offset()
   )
   n6 = Get delta time x()
   > Set WPO Regional Function!(
    execute
    Material:Material Instance Dynamic Object Reference=Get Mat Instance Dynamic()
    Overlay Material:Material Instance Dynamic Object Reference=Get Overlay Mat Instance Dynamic()
    Skeletal Mesh:Skeletal Mesh Component Object Reference=Get skeletal mesh()
    Velocity Transform:Name=foot_r
    Velocity Interp Speed:double=Get Velocity Interp Speed()
    Collision Transform:Name=calf_r
    Collision Transform Axis:EAxis Enum
    Invert Collision Transform Axis:bool=false
    Delta Time X:double=@n6
    Previous Velocity Transform Location:"Vector (by ref)"=Get R1 Velocity Transform Location()
    Previous Velocity Transform Velocity:"Vector (by ref)"=Get R1 Velocity Transform Velocity()
    Region:int=1
   ) -> (then)
   > Set WPO Regional Function(
    Material=Get Mat Instance Dynamic()
    Overlay Material=Get Overlay Mat Instance Dynamic()
    Skeletal Mesh=Get skeletal mesh()
    Velocity Transform=foot_l
    Velocity Interp Speed=Get Velocity Interp Speed()
    Collision Transform=calf_l
    Invert Collision Transform Axis=true
    Delta Time X=@n6
    Previous Velocity Transform Location=Get R2 Velocity Transform Location()
    Previous Velocity Transform Velocity=Get R2 Velocity Transform Velocity()
    Region=2
   )
   > n7 = Branch(Condition=Get bIsSkydiveFloating())
   > n8 = Set WPO Pho Wind Intensity(WPO Pho Wind Intensity=1)
   SetScalarParameterValue(execute=[@n9.then, @n8.then], self=Get Mat Instance Dynamic(), ParameterName=ClothSim_WindStrength, Value=Get WPO Pho Wind Intensity())
   n9 = Set WPO Pho Wind Intensity(execute=@n7.else, WPO Pho Wind Intensity=0)
   n10 = Sequence(execute=@n2.then_4)
   n11 = Dot Product!(
    A:Vector=GetActorRightVector!(self:Actor Object Reference=TryGetPawnOwner()) -> (ReturnValue:Vector)
    B:Vector=Normalize!(A:Vector=Get PawnVelocity(), Tolerance:float=0.0001) -> (ReturnValue:Vector) {is=title K2Node_CallFunction}
   ) -> (ReturnValue:double)
   > Set crouching left(crouching left:bool=AND Boolean(
     A="In Range (Float)"!(Value:double=@n11, Min:double=-1, Max:double=-0.44, InclusiveMin:bool=true, InclusiveMax:bool=true) -> (ReturnValue:bool)
     B=Get bIsCrouching()
    )) -> (Output_Get:bool)
   > Set crouching left float 0-5(crouching left float 0-5:double=SelectFloat(A=0.5, B=0, bPickA=Get crouching left())) -> (Output_Get:double)
   n12 = Get is helmet on()
   Set backpack move after jumping(
    execute=@n10.then_1
    backpack move after jumping:bool=AND Boolean!(A=NOT Boolean(A=@n12), B=Get bIsFalling(), C:bool=Get bIsMoving2D()) -> (ReturnValue)
   ) -> (Output_Get:bool)
   n13 = float * float(
    A="To Float (Boolean)"(InBool=Get bIsCrouchSprinting())
    B=GetAlphaFromReferenceTransformDelta(SkeletalMeshComponent=Get skeletal mesh(), BoneName=pelvis, BaseName=thigh_r, InputRange=(X=90,Y=108), TwistAxis="1,0,0")
   ) {is=title K2Node_PromotableOperator}
   Set right leg alpha combp crouching sprinting(
    execute=@n2.then_1
    right leg alpha combp crouching sprinting:double=FInterpTo(
     Current=Get right leg alpha combp crouching sprinting()
     Target="Clamp (Float)"(Value=@n13, Min=0, Max=1)
     DeltaTime=Get delta time x()
     InterpSpeed=20
    )
   ) -> (Output_Get:double)
   n14 = Get bIsPlayingEmote()
   > n15 = Branch(Condition=AND Boolean(A=Get bIsCrouching(), B=Get bIsMoving2D(), C=NOT Boolean(A=@n14)))
   n16 = Get skeletal mesh()
   > Set left leg alpha combo crouching sprinting(left leg alpha combo crouching sprinting:double=FInterpTo(
     Current=Get left leg alpha combo crouching sprinting()
     Target=GetAlphaFromReferenceTransformDelta(SkeletalMeshComponent=@n16, BoneName=pelvis, BaseName=thigh_l, InputRange=(X=50,Y=105), TwistAxis="1,0,0")
     DeltaTime=Get delta time x()
     InterpSpeed=30
    )) -> (Output_Get:double)
   Set left leg alpha combo crouching sprinting(execute=@n15.else, left leg alpha combo crouching sprinting=0)
   n17 = Get bIsPlayingEmote()
   > n18 = Branch(Condition=AND Boolean!(A=Get bIsCrouching(), B=NOT Boolean(A=@n17)) -> (ReturnValue))
   n19 = Get skeletal mesh()
   > Set left leg alpha combo crouching sitting(left leg alpha combo crouching sitting:double=FInterpTo(
     Current=Get left leg alpha combo crouching sitting()
     Target=GetAlphaFromReferenceTransformDelta(SkeletalMeshComponent=@n19, BoneName=pelvis, BaseName=thigh_l, InputRange=(X=90,Y=108), TwistAxis="1,0,0")
     DeltaTime=Get delta time x()
     InterpSpeed=60
    )) -> (Output_Get:double)
   Set left leg alpha combo crouching sitting(execute=@n18.else, left leg alpha combo crouching sitting=0)
   Event BlueprintInitializeAnimation()
   n20 = TryGetPawnOwner()
   > Is Valid(InputObject=@n20)
   > n21 = Cast To FortPlayerPawnAthena(Object=@n20)
   > Set PlayerPawn(PlayerPawn:Fort Player Pawn Athena Object Reference=@n21) -> (Output_Get:Fort Player Pawn Athena Object Reference)
   Event OnCharacterCustomizationLoaded()
   > n22 = Set skeletal mesh(skeletal mesh:Skeletal Mesh Component Object Reference=GetPartSkeletalMeshComponent()) -> (Output_Get:Skeletal Mesh Component Object Reference)
   > n23 = Get Dynamic Material Instance(SkeletalMeshComponent=@n22)
   > Set Mat Instance Dynamic(Mat Instance Dynamic=@n23)
   > n24 = GetOrCreateDynamicOverlayMaterial(MeshComponent=@n22, MaterialName=None)
   > Set Overlay Mat Instance Dynamic(Overlay Mat Instance Dynamic=@n24)
   > n25 = Cast To FortPlayerPawn(Object=TryGetPawnOwner())
   n26 = GetSkeletalMeshForPartType(self=@n25, PartType=Face)
   > n27 = Branch(Condition=Contains(SearchIn=GetDisplayName(Object=@n26), Substring=Helmet))
   > Set is helmet on(is helmet on:bool=true) -> (Output_Get:bool)
   Set is helmet on(execute=@n27.else, is helmet on=false)
   n28 = Dot Product(A=GetActorRightVector(self=TryGetPawnOwner()), B=Normalize(A=Get PawnVelocity(), Tolerance=0.0001))
   Set crouching left(crouching left=AND Boolean(A="In Range (Float)"(Value=@n28, Min=-1, Max=-0.44, InclusiveMin=true, InclusiveMax=true), B=Get bIsCrouching()))
   > Set crouching left float 0-5(crouching left float 0-5=SelectFloat(A=0.5, B=0, bPickA=Get crouching left()))
  ]}
]
variables [
 alpha pants {type=number}
 backpack move after jumping {type=boolean}
 crouching left {type=boolean}
 crouching left float 0-5 {type=number}
 crouching or moving 2d {type=boolean}
 delta time x {type=number}
 is helmet on {type=boolean}
 jumping or running {type=boolean}
 left leg alpha combo crouching sitting {type=number}
 left leg alpha combo crouching sprinting {type=number}
 left leg alpha_0 {type=number}
 left part belt {type=number}
 Local Noise Offset {type=Vector}
 Local Noise Transform Location {type=Vector}
 Mat Instance Dynamic {type=MaterialInstanceDynamic}
 Overlay Mat Instance Dynamic {type=MaterialInstanceDynamic}
 pants alpha dynamics {type=number}
 PlayerPawn {type=FortPlayerPawnAthena}
 R1 Velocity Transform Location {type=Vector}
 R1 Velocity Transform Velocity {type=Vector}
 R2 Velocity Transform Location {type=Vector}
 R2 Velocity Transform Velocity {type=Vector}
 right leg alpha combp crouching sprinting {type=number}
 right leg alpha_0 {type=number}
 skeletal mesh {type=SkeletalMeshComponent}
 Velocity Interp Speed {type=number}
 WPO Pho Wind Intensity {type=number}
]
